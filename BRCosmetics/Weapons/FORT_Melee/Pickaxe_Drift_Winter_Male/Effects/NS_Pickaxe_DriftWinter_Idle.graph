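# Niagara idle effect for the Drift Winter male pickaxe, giving it a pale blue frost haze and small white snowflakes.
# Frost sprites play a SubUV animation and resample the skeletal mesh each update to stay attached, drifting outward and fading, with higher opacity when FrontendSpawn is set.
# Snowflakes spawn on the mesh, then move freely under gravity and curl noise, with size and alpha rising and falling over life.
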NiagaraSystem /BRCosmetics/Weapons/FORT_Melee/Pickaxe_Drift_Winter_Male/Effects/NS_Pickaxe_DriftWinter_Idle
dependencies {
 usedDataInterfaces [/Script/Niagara.NiagaraDataInterfaceSimpleCounter, &NiagaraDataInterfaceCurve]
 usedDynamicInputs [&ModuloInt, /Niagara/DynamicInputs/DataInterface/GenerateIntFromCounter]
 usedModules [&SystemState]
}
emitters {
 Frost {
  emitterUpdateScript [
   &EmitterState()
   &SpawnRate(SpawnRate=12, Use Spawn Probability=false)
  ]
  particleSpawnScript [
   /Niagara/Modules/Spawn/Initialization/InitializeParticle(
    Color=&MakeCustomLinearColorFromBool(Bool=$User.FrontendSpawn, True Color=<rgba 0.2779 0.6803 1 0.7>, False Color=<rgba 0.2779 0.6803 1 0.1>)
    Lifetime=&RandomRangeFloat(Minimum=1.3, Maximum=2.35)
    Module.Write Color=true
    Module.Write Lifetime=true
    Module.Write Mass=true
    Module.Write Material Random=true
    Module.Write Position=true
    Module.Write Scale=false
    Module.Write SpriteRotation=true
    Module.Write SpriteSize=true
    Sprite Rotation=&RandomRangeFloat(Minimum=-180, Maximum=180)
    Sprite Size=&Vector2DFromFloat(Value=&RandomRangeFloat(Minimum=12, Maximum=15))
   )
   &AddVelocity(Velocity=<0 -0.25 0>)
  ]
  particleUpdateScript [
   &ParticleState()
   &SubUVAnimation(End Frame=40, Start Frame=2)
   &SkeletalMeshLocation(
    Inherit Velocity=false
    MaskSpawnGroup=false
    Mesh Sampling Type=::NewEnumerator5
    Sampled Position Offset=&Add_Vector(
     A=&RandomVector(Vector Scale=5)
     B=&MakeVector(
      X=&FloatFromCurve(
       FloatCurve=DataInterface() {propertyValues {Curve {keys [<at 0 0>,<at 1 30>]}}}
       Scale Curve=&RandomRangeFloat(Minimum=0.4, Maximum=1.2)
      )
      Y=0
      Z=0
     )
    )
    Sampled Position Offset Coordinate Space:ENiagaraMeshTransforms=::NewEnumerator3
    Skeletal Mesh=$User.SkeletalMesh
    Socket Sampling Mode=::NewEnumerator5
   )
   &SolveForcesAndVelocity(Clamp Velocity=false, Limit Acceleration=false)
   &ScaleColor(
    Scale Alpha=&FloatFromCurve(FloatCurve=DataInterface() {propertyValues {Curve {keys [
         {arriveTangent=0.1402, interpMode=RCIM_Cubic, leaveTangent=0.1402, tangentMode=RCTM_User, time=0, value=0}
         {arriveTangent=0.317, interpMode=RCIM_Cubic, leaveTangent=0.317, tangentMode=RCTM_User, time=0.4, value=0.4}
         <at 0.7886 0 RCIM_Cubic>
        ]}}})
    ScaleA=true
    ScaleRGB=true
   )
   &CameraOffset(Camera Offset Amount=&FloatFromCurve(FloatCurve=DataInterface() {propertyValues {Curve {keys [<at 0 1>,<at 1 20>]}}}))
   &ScaleSpriteSize(Scale Factor=&Vector2DFromFloat(Value=&FloatFromCurve(FloatCurve=DataInterface() {propertyValues {Curve {keys [<at 0 1.5 RCIM_Cubic>, {interpMode=RCIM_Cubic, leaveTangent=0.8108, time=1, value=0}]}}})))
   &DynamicMaterialParameters(
    Index 0 Param 4=1
    Module.Param0WriteEnabled=true
    Module.Param1WriteEnabled=true
    Module.Param2WriteEnabled=true
    Module.Param3WriteEnabled=true
   )
  ]
  rendererClasses [&NiagaraSpriteRendererProperties]
  renderers [{rendererClass=&NiagaraSpriteRendererProperties}]
 }
 snowflake_Local {
  emitterUpdateScript [
   &EmitterState()
   &SpawnRate(SpawnRate=20, Use Spawn Probability=false)
  ]
  particleSpawnScript [
   &InitializeParticle(
    Color=&MakeLinearColorFromVectorAndFloat("Vector (RGB)"=<1 1 1>, "Float (Alpha)"=&RandomRangeFloat(Minimum=1, Maximum=1))
    Module.Write Color=true
    Module.Write Lifetime=true
    Module.Write Mass=true
    Module.Write Material Random=true
    Module.Write Position=true
    Module.Write Scale=false
    Module.Write SpriteRotation=false
    Module.Write SpriteSize=true
    Sprite Size=&Vector2DFromFloat(Value=&RandomRangeFloat(Minimum=0.2, Maximum=3))
   )
   &CameraOffset(Camera Offset Amount=2)
   &SkeletalMeshLocation(
    Inherit Velocity=false
    MaskSpawnGroup=false
    Mesh Sampling Type=::NewEnumerator8
    Module.Override Randomness=false
    OffsetAlongNormal=false
    Skeletal Mesh=$User.SkeletalMesh
   )
   /SetVariables(Particles.Position=&Add_Vector(
     A=$Output.SkeletalMeshLocation.SampledPosition
     B=&RandomRangeVector(Minimum=<-2 -2 -2>, Maximum=<2 2 2>)
    )) {enabled=false, isSetParameters=true}
  ]
  particleUpdateScript [
   &ParticleState()
   &AddVelocity(Coordinate Space=::World, Velocity=&Multiply_Vector(A=$Particles.Nolmailze, B=<0 1 0>))
   &GravityForce(Gravity=<0 0 -15>)
   &CurlNoiseForce(Pan Noise=false)
   &SolveForcesAndVelocity(Clamp Velocity=false, Limit Acceleration=false)
   &ScaleSpriteSize(Scale Factor=&Vector2DFromFloat(Value=&FloatFromCurve(FloatCurve=DataInterface() {propertyValues {Curve {keys [
          {arriveTangent=4.9073, interpMode=RCIM_Cubic, leaveTangent=4.9073, tangentMode=RCTM_User, time=0, value=0}
          <at 0.2737 1 RCIM_Cubic>
          {arriveTangent=-2.1114, interpMode=RCIM_Cubic, leaveTangent=-2.1114, tangentMode=RCTM_User, time=1, value=0}
         ]}}})))
   &ScaleColor(
    Scale Alpha=&FloatFromCurve(FloatCurve=DataInterface() {propertyValues {Curve {keys [
         {arriveTangent=1.2238, interpMode=RCIM_Cubic, leaveTangent=1.2238, tangentMode=RCTM_User, time=0, value=0.3}
         <at 0.4323 0.9 RCIM_Cubic>
         <at 1 0.3 RCIM_Cubic>
        ]}}})
    ScaleA=true
    ScaleRGB=true
   )
  ]
  rendererClasses [&NiagaraSpriteRendererProperties]
  renderers [{rendererClass=&NiagaraSpriteRendererProperties}]
 }
}
summary {
 systemName=NS_Pickaxe_DriftWinter_Idle
 emitters [
  {emitterName=Frost, rendererClasses [&NiagaraSpriteRendererProperties]}
  {emitterName=snowflake_Local, rendererClasses [&NiagaraSpriteRendererProperties]}
 ]
 userVariables [
  User.bIsFrontEndPreview {type=NiagaraBool}
  User.FrontendSpawn {type=NiagaraBool}
  User.SkeletalMesh {
   type=NiagaraDataInterfaceSkeletalMesh
   defaultValue {
    struct=NiagaraExt_VariableValue_DataInterface
    value {dataInterface=&self:NiagaraDataInterfaceSkeletalMesh_1}
   }
  }
 ]
}
