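# Physics asset for the Drift Winter Male pickaxe, a two-handed melee weapon cosmetic.
# It puts small capsule bodies on the handle bone and on seven dyn_ambient_master_jnt bones, with the joint capsules rotated 90 degrees.
# Each joint is constrained directly to the handle in a star layout rather than a chain.
# All swing and twist axes are locked, so the 45-degree limit values allow no angular motion.
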PhysicsAsset /BRCosmetics/Weapons/FORT_Melee/Pickaxe_Drift_Winter_Male/Meshes/Drift_Winter_Male_Physics
bodies {
 Melee_TwoHanded_Handle [{shapeType=Capsule, length=2, radius=2}]
 dyn_ambient_master_jnt_1 [{shapeType=Capsule, length=2, radius=1, rotation=<rot 0 0 90>}]
 dyn_ambient_master_jnt_2 [{shapeType=Capsule, length=7, radius=1.5, rotation=<rot 0 0 90>}]
 dyn_ambient_master_jnt_3 [{shapeType=Capsule, length=2, radius=1, rotation=<rot 0 0 90>}]
 dyn_ambient_master_jnt_4 [{shapeType=Capsule, length=2, radius=1, rotation=<rot 0 0 90>}]
 dyn_ambient_master_jnt_5 [{shapeType=Capsule, length=2, radius=1, rotation=<rot 0 0 90>}]
 dyn_ambient_master_jnt_6 [{shapeType=Capsule, length=4, radius=1.5, rotation=<rot 0 0 90>}]
 dyn_ambient_master_jnt_7 [{shapeType=Capsule, length=2, radius=1, rotation=<rot 0 0 90>}]
}
constraints [
 {
  bone1Name=dyn_ambient_master_jnt_7
  bone2Name=Melee_TwoHanded_Handle
  swing1LimitDegrees=45
  swing1Motion=Locked
  swing2LimitDegrees=45
  swing2Motion=Locked
  twistLimitDegrees=45
  twistMotion=Locked
 }
 {
  bone1Name=dyn_ambient_master_jnt_6
  bone2Name=Melee_TwoHanded_Handle
  swing1LimitDegrees=45
  swing1Motion=Locked
  swing2LimitDegrees=45
  swing2Motion=Locked
  twistLimitDegrees=45
  twistMotion=Locked
 }
 {
  bone1Name=dyn_ambient_master_jnt_5
  bone2Name=Melee_TwoHanded_Handle
  swing1LimitDegrees=45
  swing1Motion=Locked
  swing2LimitDegrees=45
  swing2Motion=Locked
  twistLimitDegrees=45
  twistMotion=Locked
 }
 {
  bone1Name=dyn_ambient_master_jnt_4
  bone2Name=Melee_TwoHanded_Handle
  swing1LimitDegrees=45
  swing1Motion=Locked
  swing2LimitDegrees=45
  swing2Motion=Locked
  twistLimitDegrees=45
  twistMotion=Locked
 }
 {
  bone1Name=dyn_ambient_master_jnt_3
  bone2Name=Melee_TwoHanded_Handle
  swing1LimitDegrees=45
  swing1Motion=Locked
  swing2LimitDegrees=45
  swing2Motion=Locked
  twistLimitDegrees=45
  twistMotion=Locked
 }
 {
  bone1Name=dyn_ambient_master_jnt_2
  bone2Name=Melee_TwoHanded_Handle
  swing1LimitDegrees=45
  swing1Motion=Locked
  swing2LimitDegrees=45
  swing2Motion=Locked
  twistLimitDegrees=45
  twistMotion=Locked
 }
 {
  bone1Name=dyn_ambient_master_jnt_1
  bone2Name=Melee_TwoHanded_Handle
  swing1LimitDegrees=45
  swing1Motion=Locked
  swing2LimitDegrees=45
  swing2Motion=Locked
  twistLimitDegrees=45
  twistMotion=Locked
 }
]
