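# Animation Blueprint for the Drift Winter Male pickaxe cosmetic, adding physics and procedural floating motion to its bones.
# The AnimGraph runs the input pose through a RigidBody node simulated relative to b_AR_Root, then caches it.
# Seven chained Modify Bone nodes add bone-space translation offsets to the dyn_ambient_jnt bones.
# Each update, the EventGraph drives those offsets with real-time sine waves of differing frequency, shared between some bones.
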AnimBlueprint /BRCosmetics/Weapons/FORT_Melee/Pickaxe_Drift_Winter_Male/Meshes/Drift_Winter_Male_AnimBP
target_skeleton=&/Drift_Winter_Male_Skeleton
graphs [
 AnimGraph {nodes [
   n1 = Get Translate Floaters 5() -> (Translate Floaters 5:Vector)
   n2 = "Transform (Modify) Bone - Bone: dyn_ambient_jnt_1"(ComponentPose=Local To Component(LocalPose=Use cached pose 'transformCache'()), Alpha=2, Translation=@n1, Rotation="0,0,0", Scale="1,1,1") {
    rotationMode=BMM_Additive
    scaleSpace=BCS_BoneSpace
    translationMode=BMM_Additive
    translationSpace=BCS_BoneSpace
    alphaBoolBlend {blendOption=Linear}
    boneToModify {boneName=dyn_ambient_jnt_1}
   }
   n3 = "Transform (Modify) Bone - Bone: dyn_ambient_jnt_2"(ComponentPose=@n2, Alpha=1, Translation=Get Translate Floaters 2() -> (Translate Floaters 2:Vector), Rotation="0,0,0", Scale="1,1,1") {
    rotationMode=BMM_Additive
    scaleSpace=BCS_BoneSpace
    translationMode=BMM_Additive
    translationSpace=BCS_BoneSpace
    alphaBoolBlend {blendOption=Linear}
    boneToModify {boneName=dyn_ambient_jnt_2}
   }
   n4 = "Transform (Modify) Bone - Bone: dyn_ambient_jnt_3"(ComponentPose=@n3, Alpha=2, Translation=@n1, Rotation="0,0,0", Scale="1,1,1") {
    rotationMode=BMM_Additive
    scaleSpace=BCS_BoneSpace
    translationMode=BMM_Additive
    translationSpace=BCS_BoneSpace
    alphaBoolBlend {blendOption=Linear}
    boneToModify {boneName=dyn_ambient_jnt_3}
   }
   n5 = Get Translate Floaters 1() -> (Translate Floaters 1:Vector)
   n6 = "Transform (Modify) Bone - Bone: dyn_ambient_jnt_4"(ComponentPose=@n4, Alpha=1, Translation=@n5, Rotation="0,0,0", Scale="1,1,1") {
    rotationMode=BMM_Additive
    scaleSpace=BCS_BoneSpace
    translationMode=BMM_Additive
    translationSpace=BCS_BoneSpace
    alphaBoolBlend {blendOption=Linear}
    boneToModify {boneName=dyn_ambient_jnt_4}
   }
   n7 = Get Translate Floaters 3() -> (Translate Floaters 3:Vector)
   n8 = "Transform (Modify) Bone - Bone: dyn_ambient_jnt_5"(ComponentPose=@n6, Alpha=1, Translation=@n7, Rotation="0,0,0", Scale="1,1,1") {
    rotationMode=BMM_Additive
    scaleSpace=BCS_BoneSpace
    translationMode=BMM_Additive
    translationSpace=BCS_BoneSpace
    alphaBoolBlend {blendOption=Linear}
    boneToModify {boneName=dyn_ambient_jnt_5}
   }
   n9 = "Transform (Modify) Bone - Bone: dyn_ambient_jnt_6"(ComponentPose=@n8, Alpha=2, Translation=@n5, Rotation="0,0,0", Scale="1,1,1") {
    rotationMode=BMM_Additive
    scaleSpace=BCS_BoneSpace
    translationMode=BMM_Additive
    translationSpace=BCS_BoneSpace
    alphaBoolBlend {blendOption=Linear}
    boneToModify {boneName=dyn_ambient_jnt_6}
   }
   n10 = "Transform (Modify) Bone - Bone: dyn_ambient_jnt_7"(ComponentPose=@n9, Alpha=1, Translation=@n7, Rotation="0,0,0", Scale="1,1,1") {
    rotationMode=BMM_Additive
    scaleSpace=BCS_BoneSpace
    translationMode=BMM_Additive
    translationSpace=BCS_BoneSpace
    alphaBoolBlend {blendOption=Linear}
    boneToModify {boneName=dyn_ambient_jnt_7}
   }
   Output Pose(Result=Component To Local(ComponentPose=@n10)) {layerGroup=None}
   AnimGraphNode_SubInput_0 = Input Pose()
   n11 = RigidBody(ComponentPose=Local To Component(LocalPose=@AnimGraphNode_SubInput_0)) {
    bClampLinearTranslationLimitToRefPose=true
    bOverrideWorldGravity=true
    overridePhysicsAsset=&/Drift_Winter_Male_Physics
    simulationSpace=BaseBoneSpace
    alphaBoolBlend {blendOption=Linear}
    baseBoneRef {boneName=b_AR_Root}
    componentLinearAccScale=<0.2 0.2 0.2>
    componentLinearVelScale=<0.8 0.8 0.8>
   }
   Save cached pose 'transformCache'(Pose=Component To Local(ComponentPose=@n11)) {cacheName=transformCache}
  ]}
 EventGraph {nodes [
   Event BlueprintUpdateAnimation()
   > n1 = Sequence!(execute) -> (then_0, then_1, then_2:Exec, then_3:Exec, then_4:Exec)
   n2 = "Sin (Radians)"(A=float * float(A=GetRealTimeSeconds(), B=2.1))
   > Set Translate Floaters 1(Translate Floaters 1:Vector=MakeVector(X=float / float(A=@n2, B=1.6) {is=title K2Node_CallFunction}, Y=@n2)) -> (Output_Get:Vector)
   n3 = "Sin (Radians)"(A=float * float(A=GetRealTimeSeconds(), B=1.9))
   Set Translate Floaters 2(execute=@n1.then_1, Translate Floaters 2:Vector=MakeVector(Y=float / float(A=@n3, B=3))) -> (Output_Get:Vector)
   n4 = "Sin (Radians)"(A=float * float(A=GetRealTimeSeconds(), B=1.6))
   Set Translate Floaters 3(execute=@n1.then_2, Translate Floaters 3:Vector=MakeVector(X=float / float(A=@n4, B=4), Y=@n4)) -> (Output_Get:Vector)
   n5 = float * float(A=GetRealTimeSeconds(), B=0.8)
   Set Translate Floaters 5(execute=@n1.then_4, Translate Floaters 5:Vector=MakeVector(Y="Sin (Radians)"(A=@n5))) -> (Output_Get:Vector)
   TryGetPawnOwner()
   FInterpEaseInOut!(A:double=0, B:double=0, Alpha:double=0, Exponent:double=0) -> (ReturnValue:double)
   float / float(A=0, B=1.6)
  ]}
]
variables [
 Translate Floaters 1 {type=Vector}
 Translate Floaters 2 {type=Vector}
 Translate Floaters 3 {type=Vector}
 Translate Floaters 4 {type=Vector}
 Translate Floaters 5 {type=Vector}
]
